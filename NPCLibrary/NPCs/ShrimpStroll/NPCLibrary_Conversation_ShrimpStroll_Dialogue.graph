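# Conversation database for the NPC ShrimpStroll, providing an "Ask about..." dialogue branch from a single tagged entry point.
# The opening speech offers a topiaries topic, a scissors-paper-stone question and a Back option.
# Topiaries plays a three-line chain about cutting plants into shapes for friends, while the game question gets one puzzled reply.
# Both topics converge on a shared ReturnToStart node, shown as BACK.
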ConversationDatabase /NPCLibrary/NPCs/ShrimpStroll/NPCLibrary_Conversation_ShrimpStroll_Dialogue
entry_points [{tag=Conversation.Entry.ShrimpStroll.Dialogue, destinations [#1]}]
nodes [
 ConversationEntryPointNode_0 {
  id=#1
  class=ConversationEntryPointNode
  outputs [#2]
  properties {entryTag {tagName=Conversation.Entry.ShrimpStroll.Dialogue}}
 }
 Speech_C_0 {
  id=#2
  class=Speech_C
  outputs [#3, #4, #5]
  properties {nodeName="", softTaskIcon=&T_UI_NPC_question, generalConfig {defaultMessage=...}}
  sub_nodes [ConversationChoiceNode_5 {class=ConversationChoiceNode, properties {defaultChoiceDisplayText=Ask about...}}]
 }
 Speech_C_5 {
  id=#5
  class=Speech_C
  outputs [#6]
  properties {nodeName="", softTaskIcon=&T_UI_NPC_question, generalConfig {defaultMessage=I'm good at cutting plants.}}
  sub_nodes [ConversationChoiceNode_6 {class=ConversationChoiceNode, properties {defaultChoiceDisplayText=Topiaries}}]
 }
 Speech_C_6 {
  id=#6
  class=Speech_C
  outputs [#7]
  properties {nodeName="", softTaskIcon=&T_UI_NPC_question, generalConfig {defaultMessage=So I cut them into shapes.}}
 }
 Speech_C_7 {
  id=#7
  class=Speech_C
  outputs [#8]
  properties {nodeName="", softTaskIcon=&T_UI_NPC_question, generalConfig {defaultMessage=For my friends.}}
 }
 ReturnToStart_C_2 {
  id=#8
  class=ReturnToStart_C
  properties {softTaskIcon=&T_UI_NPC_Back}
  sub_nodes [ConversationChoiceNode_4 {class=ConversationChoiceNode, properties {defaultChoiceDisplayText=BACK}}]
 }
 Back_C_1 {
  id=#4
  class=Back_C
  sub_nodes [ConversationChoiceNode_0 {class=ConversationChoiceNode, properties {defaultChoiceDisplayText=BACK}}]
 }
 Speech_C_3 {
  id=#3
  class=Speech_C
  outputs [#8]
  properties {nodeName="", softTaskIcon=&T_UI_NPC_question, generalConfig {defaultMessage=Play scissors what?}}
  sub_nodes [ConversationChoiceNode_7 {class=ConversationChoiceNode, properties {defaultChoiceDisplayText="Play scissors, paper, stone?"}}]
 }
]
